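# Control Rig for the LethalVae back bling that procedurally animates its eye and two wings.
# When not tac sprinting, a per-item chain harmonics unit moves the eye chain, with X and Z speeds driven by seeded random floats.
# If also not DBNO, matching harmonics units then run on the right and left wing chains.
# Tac sprinting skips all motion, DBNO skips only the wings, and ear and side-chain bones are untouched.
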ControlRigBlueprint /BRCosmetics/Accessories/FORT_Backpacks/Backpack_F_MED_LethalVae/Meshes/F_MED_LethalVae_Pack_CtrlRig
graphs [
 RigVMFunctionLibrary {links [], nodes []}
 RigVMModel {nodes [
   RigUnit_BeginExecution = &RigUnit_BeginExecution()
   RigVMFunction_ControlFlowBranch = &RigVMFunction_ControlFlowBranch(ExecuteContext=@RigUnit_BeginExecution.ExecutePin, Condition=RigVMVariableNode() {variable=Is Tac Sprinting Bool}.Value)
   Random = /Script/RigVM.RigVMFunction_RandomFloat(Duration=1, Maximum=3, Minimum=0, Seed=217) {is=struct RigVMUnitNode, template="Random::Execute(in Seed,in Minimum,in Maximum,in Duration,out Result)"}
   n1 = /Script/ControlRig.RigUnit_ChainHarmonicsPerItem(
    ChainRoot=(Type=Bone,Name="eye_harmonics_jnt")
    DrawWorldOffset=(Rotation=(X=0,Y=0,Z=0,W=1),Translation=(X=0,Y=0,Z=0),Scale3D=(X=1,Y=1,Z=1))
    Pendulum=(bEnabled=True,PendulumStiffness=2,PendulumGravity=(X=0,Y=0,Z=0),PendulumBlend=0.75,PendulumDrag=0.98,PendulumMaximum=0.1,UnwindAxis=(X=0,Y=1,Z=0),UnwindMinimum=0.2,UnwindMaximum=0.05)
    Reach=(bEnabled=True,ReachTarget=(X=0,Y=0,Z=0),ReachAxis=(X=1,Y=0,Z=0))
    Speed=(X=2,Y=0,Z=2)
    Wave=(bEnabled=True,WaveFrequency=(X=1,Y=0,Z=1),WaveAmplitude=(X=0.2,Y=0,Z=0.5),WaveOffset=(X=0,Y=0,Z=0),WaveNoise=(X=0.2,Y=0,Z=0.2),WaveMinimum=1,WaveMaximum=3,WaveEase=SineEaseInOut)
    WaveCurve=(EditorCurveData=(Keys=((),(Time=1,Value=1)),DefaultValue=340282346638528859811704183484516925440,PreInfinityExtrap=RCCE_Constant,PostInfinityExtrap=RCCE_Constant))
    bDrawDebug=true
    Speed.Z=@Random.Result
    Speed.X=&RigVMFunction_RandomFloat(Duration=0.1, Maximum=3, Minimum=0, Seed=217) {template="Random::Execute(in Seed,in Minimum,in Maximum,in Duration,out Result)"}.Result
    ExecutePin=@RigVMFunction_ControlFlowBranch.False
   ) {is=struct RigVMUnitNode}
   n2 = &RigVMFunction_ControlFlowBranch(ExecuteContext=@n1.ExecutePin, Condition=RigVMVariableNode() {variable=Is DBNO Bool}.Value)
   &RigUnit_ChainHarmonicsPerItem(
    ChainRoot=(Type=Bone,Name="master_R_wing_jnt")
    DrawWorldOffset=(Rotation=(X=0,Y=0,Z=0,W=1),Translation=(X=0,Y=0,Z=0),Scale3D=(X=1,Y=1,Z=1))
    Pendulum=(bEnabled=True,PendulumStiffness=2,PendulumGravity=(X=0,Y=0,Z=0),PendulumBlend=0.75,PendulumDrag=0.98,PendulumMaximum=0.1,UnwindAxis=(X=0,Y=1,Z=0),UnwindMinimum=0.2,UnwindMaximum=0.05)
    Reach=(bEnabled=True,ReachTarget=(X=0,Y=0,Z=0),ReachAxis=(X=1,Y=0,Z=0))
    Speed=(X=0,Y=1,Z=0.8)
    Wave=(bEnabled=True,WaveFrequency=(X=0,Y=0,Z=0),WaveAmplitude=(X=0,Y=0.2,Z=0.1),WaveOffset=(X=0,Y=0,Z=0),WaveNoise=(X=0,Y=3,Z=3),WaveMinimum=1,WaveMaximum=2,WaveEase=SineEaseInOut)
    WaveCurve=(EditorCurveData=(Keys=((),(Time=1,Value=1)),DefaultValue=340282346638528859811704183484516925440,PreInfinityExtrap=RCCE_Constant,PostInfinityExtrap=RCCE_Constant))
    bDrawDebug=True
    ExecutePin=@n2.False
   )
   > &RigUnit_ChainHarmonicsPerItem(
    ChainRoot=(Type=Bone,Name="master_L_wing_jnt")
    DrawWorldOffset=(Rotation=(X=0,Y=0,Z=0,W=1),Translation=(X=0,Y=0,Z=0),Scale3D=(X=1,Y=1,Z=1))
    Pendulum=(bEnabled=True,PendulumStiffness=2,PendulumGravity=(X=0,Y=0,Z=0),PendulumBlend=0.75,PendulumDrag=0.98,PendulumMaximum=0.1,UnwindAxis=(X=0,Y=1,Z=0),UnwindMinimum=0.2,UnwindMaximum=0.05)
    Reach=(bEnabled=True,ReachTarget=(X=0,Y=0,Z=0),ReachAxis=(X=1,Y=0,Z=0))
    Speed=(X=0,Y=1,Z=0.8)
    Wave=(bEnabled=True,WaveFrequency=(X=0,Y=0,Z=0),WaveAmplitude=(X=0,Y=0.2,Z=0.1),WaveOffset=(X=0,Y=0,Z=0),WaveNoise=(X=0,Y=3,Z=3),WaveMinimum=1,WaveMaximum=2,WaveEase=SineEaseInOut)
    WaveCurve=(EditorCurveData=(Keys=((),(Time=1,Value=1)),DefaultValue=340282346638528859811704183484516925440,PreInfinityExtrap=RCCE_Constant,PostInfinityExtrap=RCCE_Constant))
    bDrawDebug=True
   )
  ]}
]
hierarchy [
 root {loc=<0 0 0>, rot=<0 0 0 1>}
 pack {parent=root, loc=<0 -0.7316 0>, rot=<0 0 0 1>}
 master_sim_jnt {parent=pack, loc=<0 -5.3457 0>, rot=<0 0 0 1>}
 master_R_chain_jnt {parent=master_sim_jnt, loc=<-10.3207 1.05 -3.4175>, rot=<0.1049 -0.6036 0.2759 -0.7406>}
 dyn_R_chain_1_jnt {parent=master_R_chain_jnt, loc=<1.7639 -0.433 0.2164>, rot=<-0.0576 0.1996 0.1211 -0.9707>}
 dyn_R_chain_2_jnt {parent=dyn_R_chain_1_jnt, loc=<5.7865 1.0028 0.7318>, rot=<-0.1168 0.0391 -0.3088 -0.9431>}
 dyn_R_chain_3_jnt {parent=dyn_R_chain_2_jnt, loc=<4.1103 0.1975 1.399>, rot=<-0.0227 0.2296 -0.1189 -0.9657>}
 master_L_chain_jnt {parent=master_sim_jnt, loc=<10.3207 1.05 -3.4175>, rot=<0.7406 0.2759 0.6036 0.1049>}
 dyn_L_chain_1_jnt {parent=master_L_chain_jnt, loc=<-1.7639 0.433 -0.2164>, rot=<-0.0576 0.1996 0.1211 -0.9707>}
 dyn_L_chain_2_jnt {parent=dyn_L_chain_1_jnt, loc=<-5.7865 -1.0028 -0.7318>, rot=<-0.1168 0.0391 -0.3088 -0.9431>}
 dyn_L_chain_3_jnt {parent=dyn_L_chain_2_jnt, loc=<-4.1103 -0.1975 -1.399>, rot=<-0.0227 0.2296 -0.1189 -0.9657>}
 master_R_ear_jnt {parent=master_sim_jnt, loc=<-5.0166 1.5365 6.9169>, rot=<0.0017 -0.9337 -0.0007 0.3581>}
 dyn_R_ear_jnt {parent=master_R_ear_jnt, loc=<0.5358 0 0>, rot=<0 0 0 1>}
 master_L_ear_jnt {parent=master_sim_jnt, loc=<5.0166 1.5365 6.9169>, rot=<-0.3581 -0.0007 0.9337 0.0017>}
 dyn_L_ear_jnt {parent=master_L_ear_jnt, loc=<-0.5358 0 0>, rot=<0 0 0 1>}
 master_R_wing_jnt {parent=master_sim_jnt, loc=<-7.3604 1.1546 3.4009>, rot=<0.3414 0.0196 -0.7556 -0.5587>}
 dyn_R_wing_2_jnt {parent=master_R_wing_jnt, loc=<-0.6367 0.0335 2.368>, rot=<0.143 -0.0388 -0.0365 -0.9883>}
 dyn_R_wing_3_jnt {parent=dyn_R_wing_2_jnt, loc=<-0.6784 0.689 3.3994>, rot=<0.1916 -0.0127 0.0008 -0.9814>}
 dyn_R_wing_4_jnt {parent=dyn_R_wing_3_jnt, loc=<-0.6504 -0.0495 3.2939>, rot=<-0.2945 0.3148 -0.003 -0.9023>}
 dyn_R_wing_5_jnt {parent=dyn_R_wing_4_jnt, loc=<-1.9856 -0.3997 3.073>, rot=<-0.0207 0.4315 -0.1047 -0.8958>}
 dyn_R_wing_6_jnt {parent=dyn_R_wing_5_jnt, loc=<-0.8842 -0.2241 0.7991>, rot=<-0.0217 0.1163 -0.0169 -0.9928>}
 master_L_wing_jnt {parent=master_sim_jnt, loc=<7.3604 1.1546 3.4009>, rot=<0.5587 -0.7556 -0.0196 0.3414>}
 dyn_L_wing_2_jnt {parent=master_L_wing_jnt, loc=<0.6367 -0.0335 -2.368>, rot=<0.143 -0.0388 -0.0365 -0.9883>}
 dyn_L_wing_3_jnt {parent=dyn_L_wing_2_jnt, loc=<0.6784 -0.689 -3.3994>, rot=<0.1916 -0.0127 0.0008 -0.9814>}
 dyn_L_wing_4_jnt {parent=dyn_L_wing_3_jnt, loc=<0.6504 0.0495 -3.2939>, rot=<-0.2945 0.3148 -0.003 -0.9023>}
 dyn_L_wing_5_jnt {parent=dyn_L_wing_4_jnt, loc=<1.9856 0.3997 -3.073>, rot=<-0.0207 0.4315 -0.1047 -0.8958>}
 dyn_L_wing_6_jnt {parent=dyn_L_wing_5_jnt, loc=<0.8842 0.2241 -0.7991>, rot=<-0.0217 0.1163 -0.0169 -0.9928>}
 eye_harmonics_jnt {parent=pack, loc=<0 -5.3912 0.0646>, rot=<0 0 0 1>}
 eye_root_jnt {parent=eye_harmonics_jnt, loc=<0 -1.8741 -0.0005>, rot=<0 0 0 1>}
 eye_jnt {parent=eye_root_jnt, loc=<0 -0.3708 0>, rot=<0 0 0 1>}
]
variables [
 Is DBNO Bool {type=bool, default=False}
 Is Tac Sprinting Bool {type=bool, default=False}
]
