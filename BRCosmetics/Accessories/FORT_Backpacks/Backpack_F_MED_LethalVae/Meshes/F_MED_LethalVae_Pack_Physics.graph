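# Physics asset for the LethalVae backpack cosmetic, giving capsule bodies and joints to its wings, ears and chains.
# Wing, ear and side-chain masters hang off master_sim_jnt, which joins the pack body, through angularly free constraints.
# Dynamic bones link in series with angular limits, and wing swing widens toward the tips.
# Both side-chain ends lock angularly to the middle chain bone, whose master has no constraint to the pack or sim root.
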PhysicsAsset /BRCosmetics/Accessories/FORT_Backpacks/Backpack_F_MED_LethalVae/Meshes/F_MED_LethalVae_Pack_Physics
bodies {
 dyn_L_chain_1_jnt [{
   shapeType=Capsule
   length=3
   radius=1.5
   shapeName=dyn_L_chain_1_jnt_capsule
   center=<-1.1776 0 0.7485>
   rotation=<rot 90 -179.9899 -179.991>
  }]
 dyn_L_chain_2_jnt [{
   shapeType=Capsule
   length=2
   radius=1.5
   shapeName=dyn_L_chain_2_jnt_capsule
   center=<0.146 0.8 0.4895>
   rotation=<rot 80.0001 -179.9985 -179.9988>
  }]
 dyn_L_chain_3_jnt [{
   shapeType=Capsule
   length=2
   radius=1.5
   shapeName=dyn_L_chain_3_jnt_capsule
   center=<0.5302 0.3072 0.0935>
   rotation=<rot 80 -179.9981 -179.9982>
  }]
 dyn_L_ear_jnt [{shapeType=Capsule, length=2, radius=3, shapeName=dyn_L_ear_jnt_capsule, center=<-2.1867 0 -0.3856>, rotation=<rot -80 -0.0004 0.0004>}]
 dyn_L_wing_2_jnt [{
   shapeType=Capsule
   length=4
   radius=2.5
   shapeName=dyn_L_wing_2_jnt_capsule
   center=<-3.6444 -0.6328 -0.1116>
   rotation=<rot -9.8464 -79.8492 -91.7541>
  }]
 dyn_L_wing_3_jnt [{
   shapeType=Capsule
   length=4
   radius=2.5
   shapeName=dyn_L_wing_3_jnt_capsule
   center=<-3.6445 -0.6328 -0.1116>
   rotation=<rot -9.8465 -79.8493 -91.7539>
  }]
 dyn_L_wing_4_jnt [{
   shapeType=Capsule
   length=4
   radius=2.5
   shapeName=dyn_L_wing_4_jnt_capsule
   center=<-3.6444 -0.6328 -0.1116>
   rotation=<rot -9.8462 -79.849 -91.7541>
  }]
 dyn_L_wing_5_jnt [{
   shapeType=Capsule
   length=6
   radius=2.5
   shapeName=dyn_L_wing_5_jnt_capsule
   center=<-4.0159 -0.9198 1.1386>
   rotation=<rot -9.8461 -79.8487 -81.754>
  }]
 dyn_L_wing_6_jnt [{
   shapeType=Capsule
   length=4
   radius=2.5
   shapeName=dyn_L_wing_6_jnt_capsule
   center=<-3.6444 -0.6329 -0.1116>
   rotation=<rot -9.8462 -79.8486 -91.754>
  }]
 dyn_M_chain_01_jnt [{shapeType=Capsule, length=1.5, radius=1.5, shapeName=dyn_M_chain_01_jnt_capsule, center=<9.3238 0 0.1318>, rotation=<rot -0.0001 0 90>}]
 dyn_R_chain_1_jnt [{shapeType=Capsule, length=3, radius=1.5, shapeName=dyn_R_chain_1_jnt_capsule, center=<1.1183 -0.595 -0.7>, rotation=<rot 80 0.0004 -0.0002>}]
 dyn_R_chain_2_jnt [{
   shapeType=Capsule
   length=2
   radius=1.5
   shapeName=dyn_R_chain_2_jnt_capsule
   center=<-0.2843 -0.9269 -0.2891>
   rotation=<rot 80.0001 -179.9985 -179.9988>
  }]
 dyn_R_chain_3_jnt [{
   shapeType=Capsule
   length=1.5
   radius=1.5
   shapeName=dyn_R_chain_3_jnt_capsule
   center=<-0.5407 -0.3319 -0.0953>
   rotation=<rot 90 -84.1169 -84.1177>
  }]
 dyn_R_ear_jnt [{shapeType=Capsule, length=2, radius=3, shapeName=dyn_R_ear_jnt_capsule, center=<1.9985 0 0.3524>, rotation=<rot -80 -0.0004 0.0004>}]
 dyn_R_wing_2_jnt [{
   shapeType=Capsule
   length=4
   radius=2.5
   shapeName=dyn_R_wing_2_jnt_capsule
   center=<3.4276 0.5952 0.105>
   rotation=<rot -9.8464 -79.8492 -91.7541>
  }]
 dyn_R_wing_3_jnt [{
   shapeType=Capsule
   length=4
   radius=2.5
   shapeName=dyn_R_wing_3_jnt_capsule
   center=<3.4276 0.5952 0.105>
   rotation=<rot -9.8465 -79.8493 -91.7539>
  }]
 dyn_R_wing_4_jnt [{
   shapeType=Capsule
   length=4
   radius=2.5
   shapeName=dyn_R_wing_4_jnt_capsule
   center=<3.4276 0.5952 0.105>
   rotation=<rot -9.8462 -79.849 -91.7541>
  }]
 dyn_R_wing_5_jnt [{
   shapeType=Capsule
   length=6.5
   radius=2.5
   shapeName=dyn_R_wing_5_jnt_capsule
   center=<4.3137 1.0026 -1.3058>
   rotation=<rot -9.8461 -79.8487 -71.754>
  }]
 dyn_R_wing_6_jnt [{
   shapeType=Capsule
   length=4
   radius=2.5
   shapeName=dyn_R_wing_6_jnt_capsule
   center=<3.4276 0.5952 0.105>
   rotation=<rot -9.8462 -79.8486 -91.754>
  }]
 master_L_chain_jnt [{shapeType=Capsule, length=0.505, radius=0.505, shapeName=master_L_chain_jnt_capsule, rotation=<rot 0 -90 0>}]
 master_L_ear_jnt [{shapeType=Capsule, length=0.505, radius=0.505, shapeName=master_L_ear_jnt_capsule, rotation=<rot 0 -90 0>}]
 master_L_wing_jnt [{shapeType=Capsule, length=0.505, radius=0.505, shapeName=master_L_wing_jnt_capsule, rotation=<rot 0 -90 0>}]
 master_M_chain_jnt [{shapeType=Capsule, length=0, radius=0.5, shapeName=master_M_chain_jnt_capsule}]
 master_R_chain_jnt [{shapeType=Capsule, length=0.505, radius=0.505, shapeName=master_R_chain_jnt_capsule, rotation=<rot 0 -90 0>}]
 master_R_ear_jnt [{shapeType=Capsule, length=0.505, radius=0.505, shapeName=master_R_ear_jnt_capsule, rotation=<rot 0 -90 0>}]
 master_R_wing_jnt [{shapeType=Capsule, length=0.505, radius=0.505, shapeName=master_R_wing_jnt_capsule, rotation=<rot 0 -90 0>}]
 master_sim_jnt [{shapeType=Capsule, length=0.505, radius=0.505, shapeName=master_sim_jnt_capsule, rotation=<rot 0 -90 0>}]
 pack [{
   shapeType=Capsule
   length=8.8473
   radius=10.9144
   shapeName=pack_capsule
   center=<-0.0038 0.1787 -6.1386>
   rotation=<rot -0.0002 -89.9793 -179.9156>
  }]
}
constraints [
 {
  bone1Name=master_L_chain_jnt
  bone2Name=master_sim_jnt
  swing1LimitDegrees=45
  swing1Motion=Free
  swing2LimitDegrees=45
  swing2Motion=Free
  twistLimitDegrees=45
  twistMotion=Free
 }
 {bone1Name=dyn_L_chain_1_jnt, bone2Name=master_L_chain_jnt, swing1LimitDegrees=45, swing2LimitDegrees=20, twistLimitDegrees=0}
 {bone1Name=dyn_L_chain_2_jnt, bone2Name=dyn_L_chain_1_jnt, swing1LimitDegrees=45, swing2LimitDegrees=20, twistLimitDegrees=0}
 {bone1Name=dyn_L_chain_3_jnt, bone2Name=dyn_L_chain_2_jnt, swing1LimitDegrees=45, swing2LimitDegrees=20, twistLimitDegrees=0}
 {
  bone1Name=master_L_ear_jnt
  bone2Name=master_sim_jnt
  swing1LimitDegrees=45
  swing1Motion=Free
  swing2LimitDegrees=45
  swing2Motion=Free
  twistLimitDegrees=45
  twistMotion=Free
 }
 {bone1Name=dyn_L_ear_jnt, bone2Name=master_L_ear_jnt, swing1LimitDegrees=20, swing2LimitDegrees=25, twistLimitDegrees=0}
 {
  bone1Name=master_L_wing_jnt
  bone2Name=master_sim_jnt
  swing1LimitDegrees=45
  swing1Motion=Free
  swing2LimitDegrees=45
  swing2Motion=Free
  twistLimitDegrees=45
  twistMotion=Free
 }
 {bone1Name=dyn_L_wing_2_jnt, bone2Name=master_L_wing_jnt, swing1LimitDegrees=3, swing2LimitDegrees=2, twistLimitDegrees=0}
 {bone1Name=dyn_L_wing_3_jnt, bone2Name=dyn_L_wing_2_jnt, swing1LimitDegrees=3, swing2LimitDegrees=2, twistLimitDegrees=0}
 {bone1Name=dyn_L_wing_4_jnt, bone2Name=dyn_L_wing_3_jnt, swing1LimitDegrees=3, swing2LimitDegrees=12, twistLimitDegrees=0}
 {bone1Name=dyn_L_wing_5_jnt, bone2Name=dyn_L_wing_4_jnt, swing1LimitDegrees=3, swing2LimitDegrees=20, twistLimitDegrees=0}
 {bone1Name=dyn_L_wing_6_jnt, bone2Name=dyn_L_wing_5_jnt, swing1LimitDegrees=3, swing2LimitDegrees=20, twistLimitDegrees=0}
 {
  bone1Name=master_R_chain_jnt
  bone2Name=master_sim_jnt
  swing1LimitDegrees=45
  swing1Motion=Free
  swing2LimitDegrees=45
  swing2Motion=Free
  twistLimitDegrees=45
  twistMotion=Free
 }
 {bone1Name=dyn_R_chain_1_jnt, bone2Name=master_R_chain_jnt, swing1LimitDegrees=45, swing2LimitDegrees=20, twistLimitDegrees=0}
 {bone1Name=dyn_R_chain_2_jnt, bone2Name=dyn_R_chain_1_jnt, swing1LimitDegrees=45, swing2LimitDegrees=20, twistLimitDegrees=0}
 {bone1Name=dyn_R_chain_3_jnt, bone2Name=dyn_R_chain_2_jnt, swing1LimitDegrees=45, swing2LimitDegrees=20, twistLimitDegrees=0}
 {
  bone1Name=master_R_ear_jnt
  bone2Name=master_sim_jnt
  swing1LimitDegrees=45
  swing1Motion=Free
  swing2LimitDegrees=45
  swing2Motion=Free
  twistLimitDegrees=45
  twistMotion=Free
 }
 {bone1Name=dyn_R_ear_jnt, bone2Name=master_R_ear_jnt, swing1LimitDegrees=20, swing2LimitDegrees=25, twistLimitDegrees=0}
 {
  bone1Name=master_R_wing_jnt
  bone2Name=master_sim_jnt
  swing1LimitDegrees=45
  swing1Motion=Free
  swing2LimitDegrees=45
  swing2Motion=Free
  twistLimitDegrees=45
  twistMotion=Free
 }
 {bone1Name=dyn_R_wing_2_jnt, bone2Name=master_R_wing_jnt, swing1LimitDegrees=3, swing2LimitDegrees=2, twistLimitDegrees=0}
 {bone1Name=dyn_R_wing_3_jnt, bone2Name=dyn_R_wing_2_jnt, swing1LimitDegrees=3, swing2LimitDegrees=2, twistLimitDegrees=0}
 {bone1Name=dyn_R_wing_4_jnt, bone2Name=dyn_R_wing_3_jnt, swing1LimitDegrees=3, swing2LimitDegrees=12, twistLimitDegrees=0}
 {bone1Name=dyn_R_wing_5_jnt, bone2Name=dyn_R_wing_4_jnt, swing1LimitDegrees=3, swing2LimitDegrees=20, twistLimitDegrees=0}
 {bone1Name=dyn_R_wing_6_jnt, bone2Name=dyn_R_wing_5_jnt, swing1LimitDegrees=3, swing2LimitDegrees=20, twistLimitDegrees=0}
 {
  bone1Name=master_sim_jnt
  bone2Name=pack
  swing1LimitDegrees=45
  swing1Motion=Free
  swing2LimitDegrees=45
  swing2Motion=Free
  twistLimitDegrees=45
  twistMotion=Free
 }
 {bone1Name=dyn_M_chain_01_jnt, bone2Name=master_M_chain_jnt, swing1LimitDegrees=15, swing2LimitDegrees=15, twistLimitDegrees=5}
 {
  bone1Name=dyn_L_chain_3_jnt
  bone2Name=dyn_M_chain_01_jnt
  swing1LimitDegrees=45
  swing1Motion=Locked
  swing2LimitDegrees=45
  swing2Motion=Locked
  twistLimitDegrees=45
  twistMotion=Locked
 }
 {
  bone1Name=dyn_R_chain_3_jnt
  bone2Name=dyn_M_chain_01_jnt
  swing1LimitDegrees=45
  swing1Motion=Locked
  swing2LimitDegrees=45
  swing2Motion=Locked
  twistLimitDegrees=45
  twistMotion=Locked
 }
]
