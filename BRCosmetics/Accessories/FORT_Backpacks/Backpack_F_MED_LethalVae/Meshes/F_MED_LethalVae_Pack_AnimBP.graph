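# Animation Blueprint for the LethalVae backpack cosmetic that animates its wing bones and reacts to tactical sprint and DBNO states.
# A RigidBody node simulates the input pose relative to the pack bone, with reduced weight during tactical sprint.
# Two interpolated chains of additive bone rotations then pose both wings for each state, before a Control Rig receives both flags.
# The event graph refreshes float and bool copies of those flags each update.
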AnimBlueprint /BRCosmetics/Accessories/FORT_Backpacks/Backpack_F_MED_LethalVae/Meshes/F_MED_LethalVae_Pack_AnimBP
target_skeleton=&/F_MED_LethalVae_Pack_Skeleton
graphs [
 AnimGraph {nodes [
   n1 = RigidBody(ComponentPose=Local To Component(LocalPose=Use cached pose 'BaseSim'()), Alpha=SelectFloat(A=0.2, B=1, bPickA=Get bIsTacticalSprinting())) {
    bClampLinearTranslationLimitToRefPose=true
    bOverrideWorldGravity=true
    overridePhysicsAsset=&/F_MED_LethalVae_Pack_Physics
    simulationSpace=BaseBoneSpace
    alphaBoolBlend {blendOption=Linear}
    baseBoneRef {boneName=pack}
    componentAppliedLinearAccClamp=<3000 3000 3000>
    componentLinearAccScale=<0.2 0.2 0.4>
    componentLinearVelScale=<0.2 0.2 0.4>
    overrideWorldGravity {z=-980}
    simSpaceSettings {maxAngularAcceleration=1000, maxAngularVelocity=3, maxLinearAcceleration=250, maxLinearVelocity=500, velocityScaleZ=0, worldAlpha=1}
   }
   n2 = Get Is Tac Sprinting() -> (Is Tac Sprinting:double)
   n3 = "Transform (Modify) Bone - Bone: dyn_R_wing_2_jnt"(ComponentPose=@n1, Alpha=@n2, Translation="0,0,0", Rotation="25,30,96", Scale="1,1,1") {
    rotationMode=BMM_Additive
    rotationSpace=BCS_BoneSpace
    alphaBoolBlend {blendOption=Linear}
    alphaScaleBiasClamp {bInterpResult=true, interpSpeedDecreasing=3, interpSpeedIncreasing=7}
    boneToModify {boneName=dyn_R_wing_2_jnt}
    rotation=<rot 25 30 96>
   }
   n4 = "Transform (Modify) Bone - Bone: dyn_R_wing_3_jnt"(ComponentPose=@n3, Alpha=@n2, Translation="0,0,0", Rotation="15,13,31", Scale="1,1,1") {
    rotationMode=BMM_Additive
    rotationSpace=BCS_BoneSpace
    alphaBoolBlend {blendOption=Linear}
    alphaScaleBiasClamp {bInterpResult=true, interpSpeedDecreasing=3, interpSpeedIncreasing=7}
    boneToModify {boneName=dyn_R_wing_3_jnt}
    rotation=<rot 15 13 31>
   }
   n5 = "Transform (Modify) Bone - Bone: dyn_R_wing_4_jnt"(ComponentPose=@n4, Alpha=@n2, Translation="0,0,0", Rotation="0,0,50", Scale="1,1,1") {
    rotationMode=BMM_Additive
    rotationSpace=BCS_BoneSpace
    alphaBoolBlend {blendOption=Linear}
    alphaScaleBiasClamp {bInterpResult=true, interpSpeedDecreasing=3, interpSpeedIncreasing=7}
    boneToModify {boneName=dyn_R_wing_4_jnt}
    rotation {roll=50}
   }
   n6 = "Transform (Modify) Bone - Bone: dyn_R_wing_5_jnt"(ComponentPose=@n5, Alpha=@n2, Translation="0,0,0", Rotation="0,4,25", Scale="1,1,1") {
    rotationMode=BMM_Additive
    rotationSpace=BCS_BoneSpace
    alphaBoolBlend {blendOption=Linear}
    alphaScaleBiasClamp {bInterpResult=true, interpSpeedDecreasing=3, interpSpeedIncreasing=7}
    boneToModify {boneName=dyn_R_wing_5_jnt}
    rotation {roll=25, yaw=4}
   }
   n7 = "Transform (Modify) Bone - Bone: dyn_R_wing_6_jnt"(ComponentPose=@n6, Alpha=@n2, Translation="0,0,0", Rotation="20,10.5,3.6", Scale="1,1,1") {
    rotationMode=BMM_Additive
    rotationSpace=BCS_BoneSpace
    alphaBoolBlend {blendOption=Linear}
    alphaScaleBiasClamp {bInterpResult=true, interpSpeedDecreasing=3, interpSpeedIncreasing=7}
    boneToModify {boneName=dyn_R_wing_6_jnt}
    rotation=<rot 20 10.5 3.6>
   }
   n8 = "Transform (Modify) Bone - Bone: dyn_L_wing_2_jnt"(ComponentPose=@n7, Alpha=@n2, Translation="0,0,0", Rotation="25,30,96", Scale="1,1,1") {
    rotationMode=BMM_Additive
    rotationSpace=BCS_BoneSpace
    alphaBoolBlend {blendOption=Linear}
    alphaScaleBiasClamp {bInterpResult=true, interpSpeedDecreasing=3, interpSpeedIncreasing=7}
    boneToModify {boneName=dyn_L_wing_2_jnt}
    rotation=<rot 25 30 96>
   }
   n9 = "Transform (Modify) Bone - Bone: dyn_L_wing_3_jnt"(ComponentPose=@n8, Alpha=@n2, Translation="0,0,0", Rotation="15,13,31", Scale="1,1,1") {
    rotationMode=BMM_Additive
    rotationSpace=BCS_BoneSpace
    alphaBoolBlend {blendOption=Linear}
    alphaScaleBiasClamp {bInterpResult=true, interpSpeedDecreasing=3, interpSpeedIncreasing=7}
    boneToModify {boneName=dyn_L_wing_3_jnt}
    rotation=<rot 15 13 31>
   }
   n10 = "Transform (Modify) Bone - Bone: dyn_L_wing_4_jnt"(ComponentPose=@n9, Alpha=@n2, Translation="0,0,0", Rotation="10,17.5,61", Scale="1,1,1") {
    rotationMode=BMM_Additive
    rotationSpace=BCS_BoneSpace
    alphaBoolBlend {blendOption=Linear}
    alphaScaleBiasClamp {bInterpResult=true, interpSpeedDecreasing=3, interpSpeedIncreasing=7}
    boneToModify {boneName=dyn_L_wing_4_jnt}
    rotation=<rot 10 17.5 61>
   }
   n11 = "Transform (Modify) Bone - Bone: dyn_L_wing_5_jnt"(ComponentPose=@n10, Alpha=@n2, Translation="0,0,0", Rotation="0,4,25", Scale="1,1,1") {
    rotationMode=BMM_Additive
    rotationSpace=BCS_BoneSpace
    alphaBoolBlend {blendOption=Linear}
    alphaScaleBiasClamp {bInterpResult=true, interpSpeedDecreasing=3, interpSpeedIncreasing=7}
    boneToModify {boneName=dyn_L_wing_5_jnt}
    rotation {roll=25, yaw=4}
   }
   n12 = "Transform (Modify) Bone - Bone: dyn_L_wing_6_jnt"(ComponentPose=@n11, Alpha=@n2, Translation="0,0,0", Rotation="40,0,0", Scale="1,1,1") {
    rotationMode=BMM_Additive
    rotationSpace=BCS_BoneSpace
    alphaBoolBlend {blendOption=Linear}
    alphaScaleBiasClamp {bInterpResult=true, interpSpeedDecreasing=3, interpSpeedIncreasing=7}
    boneToModify {boneName=dyn_L_wing_6_jnt}
    rotation {pitch=40}
   }
   n13 = Get Is DBNO() -> (Is DBNO:double)
   n14 = "Transform (Modify) Bone - Bone: dyn_R_wing_2_jnt"(ComponentPose=@n12, Alpha=@n13, Translation="0,0,0", Rotation="18.5,60,107", Scale="1,1,1") {
    rotationMode=BMM_Additive
    rotationSpace=BCS_BoneSpace
    alphaBoolBlend {blendOption=Linear}
    alphaScaleBiasClamp {bInterpResult=true, interpSpeedDecreasing=3, interpSpeedIncreasing=7}
    boneToModify {boneName=dyn_R_wing_2_jnt}
    rotation=<rot 18.5 60 107>
   }
   n15 = "Transform (Modify) Bone - Bone: dyn_R_wing_3_jnt"(ComponentPose=@n14, Alpha=@n13, Translation="0,0,0", Rotation="7,6.5,40", Scale="1,1,1") {
    rotationMode=BMM_Additive
    rotationSpace=BCS_BoneSpace
    alphaBoolBlend {blendOption=Linear}
    alphaScaleBiasClamp {bInterpResult=true, interpSpeedDecreasing=3, interpSpeedIncreasing=7}
    boneToModify {boneName=dyn_R_wing_3_jnt}
    rotation=<rot 7 6.5 40>
   }
   n16 = "Transform (Modify) Bone - Bone: dyn_R_wing_4_jnt"(ComponentPose=@n15, Alpha=@n13, Translation="0,0,0", Rotation="0,0,50", Scale="1,1,1") {
    rotationMode=BMM_Additive
    rotationSpace=BCS_BoneSpace
    alphaBoolBlend {blendOption=Linear}
    alphaScaleBiasClamp {bInterpResult=true, interpSpeedDecreasing=3, interpSpeedIncreasing=7}
    boneToModify {boneName=dyn_R_wing_4_jnt}
    rotation {roll=50}
   }
   n17 = "Transform (Modify) Bone - Bone: dyn_R_wing_5_jnt"(ComponentPose=@n16, Alpha=@n13, Translation="0,0,0", Rotation="0,4,25", Scale="1,1,1") {
    rotationMode=BMM_Additive
    rotationSpace=BCS_BoneSpace
    alphaBoolBlend {blendOption=Linear}
    alphaScaleBiasClamp {bInterpResult=true, interpSpeedDecreasing=3, interpSpeedIncreasing=7}
    boneToModify {boneName=dyn_R_wing_5_jnt}
    rotation {roll=25, yaw=4}
   }
   n18 = "Transform (Modify) Bone - Bone: dyn_R_wing_6_jnt"(ComponentPose=@n17, Alpha=@n13, Translation="0,0,0", Rotation="0,9,3.2", Scale="1,1,1") {
    rotationMode=BMM_Additive
    rotationSpace=BCS_BoneSpace
    alphaBoolBlend {blendOption=Linear}
    alphaScaleBiasClamp {bInterpResult=true, interpSpeedDecreasing=3, interpSpeedIncreasing=7}
    boneToModify {boneName=dyn_R_wing_6_jnt}
    rotation {roll=3.2, yaw=9}
   }
   n19 = "Transform (Modify) Bone - Bone: dyn_L_wing_2_jnt"(ComponentPose=@n18, Alpha=@n13, Translation="0,0,0", Rotation="23.5,40,100", Scale="1,1,1") {
    rotationMode=BMM_Additive
    rotationSpace=BCS_BoneSpace
    alphaBoolBlend {blendOption=Linear}
    alphaScaleBiasClamp {bInterpResult=true, interpSpeedDecreasing=3, interpSpeedIncreasing=7}
    boneToModify {boneName=dyn_L_wing_2_jnt}
    rotation=<rot 23.5 40 100>
   }
   n20 = "Transform (Modify) Bone - Bone: dyn_L_wing_3_jnt"(ComponentPose=@n19, Alpha=@n13, Translation="0,0,0", Rotation="15,13,41", Scale="1,1,1") {
    rotationMode=BMM_Additive
    rotationSpace=BCS_BoneSpace
    alphaBoolBlend {blendOption=Linear}
    alphaScaleBiasClamp {bInterpResult=true, interpSpeedDecreasing=3, interpSpeedIncreasing=7}
    boneToModify {boneName=dyn_L_wing_3_jnt}
    rotation=<rot 15 13 41>
   }
   n21 = "Transform (Modify) Bone - Bone: dyn_L_wing_4_jnt"(ComponentPose=@n20, Alpha=@n13, Translation="0,0,0", Rotation="10,17.5,61", Scale="1,1,1") {
    rotationMode=BMM_Additive
    rotationSpace=BCS_BoneSpace
    alphaBoolBlend {blendOption=Linear}
    alphaScaleBiasClamp {bInterpResult=true, interpSpeedDecreasing=3, interpSpeedIncreasing=7}
    boneToModify {boneName=dyn_L_wing_4_jnt}
    rotation=<rot 10 17.5 61>
   }
   n22 = "Transform (Modify) Bone - Bone: dyn_L_wing_5_jnt"(ComponentPose=@n21, Alpha=@n13, Translation="0,0,0", Rotation="9,8,25", Scale="1,1,1") {
    rotationMode=BMM_Additive
    rotationSpace=BCS_BoneSpace
    alphaBoolBlend {blendOption=Linear}
    alphaScaleBiasClamp {bInterpResult=true, interpSpeedDecreasing=3, interpSpeedIncreasing=7}
    boneToModify {boneName=dyn_L_wing_5_jnt}
    rotation=<rot 9 8 25>
   }
   n23 = "Transform (Modify) Bone - Bone: dyn_L_wing_6_jnt"(ComponentPose=@n22, Alpha=@n13, Translation="0,0,0", Rotation="40,0,0", Scale="1,1,1") {
    rotationMode=BMM_Additive
    rotationSpace=BCS_BoneSpace
    alphaBoolBlend {blendOption=Linear}
    alphaScaleBiasClamp {bInterpResult=true, interpSpeedDecreasing=3, interpSpeedIncreasing=7}
    boneToModify {boneName=dyn_L_wing_6_jnt}
    rotation {pitch=40}
   }
   n24 = Control Rig!(
    Source=Component To Local(ComponentPose=@n23)
    Alpha
    Is Tac Sprinting Bool:bool=Get Is Tac Sprinting Bool() -> (Is Tac Sprinting Bool:bool)
    Is DBNO Bool:bool=Get Is DBNO Bool() -> (Is DBNO Bool:bool)
   ) -> (Pose) {
    alphaBoolBlend {blendOption=Linear}
    controlRigAssetReference {blueprintRigClass=&/F_MED_LethalVae_Pack_CtrlRig.F_MED_LethalVae_Pack_CtrlRig_C}
   }
   Output Pose(Result=@n24) {layerGroup=None}
   Save cached pose 'BaseSim'(Pose=Input Pose()) {cacheName=BaseSim}
  ]}
 EventGraph {nodes [
   Event BlueprintUpdateAnimation()
   n1 = Get bIsTacticalSprinting()
   > Set Is Tac Sprinting(Is Tac Sprinting:double="To Float (Boolean)"(InBool=@n1)) -> (Output_Get:double)
   n2 = Get bIsDBNO()
   > Set Is DBNO(Is DBNO:double="To Float (Boolean)"(InBool=@n2)) -> (Output_Get:double)
   > Set Is Tac Sprinting Bool(Is Tac Sprinting Bool:bool=@n1) -> (Output_Get:bool)
   > Set Is DBNO Bool(Is DBNO Bool:bool=@n2) -> (Output_Get:bool)
   TryGetPawnOwner()
  ]}
]
variables [
 Is DBNO {type=number}
 Is DBNO Bool {type=boolean}
 Is Tac Sprinting {type=number}
 Is Tac Sprinting Bool {type=boolean}
]
